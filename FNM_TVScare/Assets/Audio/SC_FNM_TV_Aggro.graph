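# SC_FNM_TV_Aggro is a sound cue for the TV scare's aggro exclamation-mark alert sound.
# A single wave player feeds the root, playing the exclamation-mark wave with boosted volume.
# It borrows Fortnite's NPC alert attenuation and aggro concurrency and routes to a critical in-game UI sound class.
# Concurrency overrides also specify one instance, limited per owner.
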SoundCue /FNM_TVScare/Assets/Audio/SC_FNM_TV_Aggro
root=WavePlayer_1
nodes [{id=WavePlayer_1, properties {soundWaveAssetPtr=/FNM_TVScare/Assets/Audio/wav/FNM_TV_Agro_Exclamation_Mark}}]
properties {
 attenuationSettings=/FNGameplayCues/Athena/Sounds/NPCs/NPC_Alert_Att
 maxDistance=9000
 soundClassObject=&Fort_UI_InGame_Critical
 volumeMultiplier=1.75
 concurrencyOverrides {bLimitToOwner=true, maxCount=1}
 concurrencySet [/FNGameplayCues/Athena/Sounds/NPCs/NPC_Aggro_Conc]
}
