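# Sound cue for the TV explosion effect in the FNM TV scare, played as an in-game SFX.
# A single wave player feeds an attenuation node that overrides spatialization with stereo spread, distance-based low-pass filtering and trace-based occlusion filtering.
# A concurrency override caps simultaneous instances, stopping the oldest when the limit is reached and applying a retrigger time.
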SoundCue /FNM_TVScare/Assets/Audio/SC_FNM_TV_Explode
root=Attenuation_0
nodes [
 {
  id=Attenuation_0
  children [WavePlayer_1]
  properties {
   bOverrideAttenuation=true
   attenuationOverrides {
    bAttenuateWithLPF=true
    bEnableOcclusion=true
    lPFFrequencyAtMax=1000
    lPFRadiusMax=2000
    lPFRadiusMin=200
    occlusionLowPassFilterFrequency=1000
    occlusionTraceChannel=ECC_GameTraceChannel6
    stereoSpread=500
   }
  }
 }
 {id=WavePlayer_1, properties {soundWaveAssetPtr=/FNM_TVScare/Assets/Audio/wav/FNM_TV_Explode}}
]
properties {
 bOverrideConcurrency=true
 maxDistance=4000
 soundClassObject=&Fort_SFX_InGame
 volumeMultiplier=1
 concurrencyOverrides {maxCount=2, resolutionRule=StopOldest, retriggerTime=1}
}
