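# Sound cue for a TV interaction sound in the FNM_TVScare content.
# A single wave player feeds the FNM_TV_Interact_02 wave straight to the output, slightly boosted and routed through the Fortnite in-game SFX sound class.
# A concurrency override caps it at a couple of simultaneous instances, stopping the oldest, with a retrigger time set.
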SoundCue /FNM_TVScare/Assets/Audio/SC_FNM_TV_Interact
root=WavePlayer_2
nodes [{id=WavePlayer_2, properties {soundWaveAssetPtr=/FNM_TVScare/Assets/Audio/wav/FNM_TV_Interact_02}}]
properties {
 bOverrideConcurrency=true
 maxDistance=2097152
 soundClassObject=&Fort_SFX_InGame
 volumeMultiplier=1.2
 concurrencyOverrides {maxCount=2, resolutionRule=StopOldest, retriggerTime=1}
}
