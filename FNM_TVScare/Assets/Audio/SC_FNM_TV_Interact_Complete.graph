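# Sound cue that plays the TV interact-complete sound effect for the FNM TV Scare content.
# A single wave player feeds the FNM_TV_Interact_Complete wave straight to the output, routed through the Fort_SFX_InGame sound class.
# It overrides concurrency to allow only a couple of simultaneous instances, stopping the oldest, with a retrigger time set.
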SoundCue /FNM_TVScare/Assets/Audio/SC_FNM_TV_Interact_Complete
root=WavePlayer_0
nodes [{id=WavePlayer_0, properties {soundWaveAssetPtr=/FNM_TVScare/Assets/Audio/wav/FNM_TV_Interact_Complete}}]
properties {
 bOverrideConcurrency=true
 maxDistance=2097152
 soundClassObject=&Fort_SFX_InGame
 volumeMultiplier=1
 concurrencyOverrides {maxCount=2, resolutionRule=StopOldest, retriggerTime=1}
}
