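# Sound Cue for the TV jump-scare approach sting in FNM_TVScare, playing a TV glitch sound effect.
# A single Wave Player feeds the output directly, boosted in volume and routed to the in-game SFX sound class, with no attenuation asset set.
# It joins the TV stings concurrency set, and also stores local concurrency values that favor stopping the oldest instance.
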SoundCue /FNM_TVScare/Assets/Audio/SC_FNM_TV_JumpScare_Approach
root=WavePlayer_2
nodes [{id=WavePlayer_2, properties {soundWaveAssetPtr=/FNM_TVScare/Assets/Audio/wav/FNM_TV_Sting_JumpScare_SFX_TVGlitch}}]
properties {
 maxDistance=2097152
 soundClassObject=&Fort_SFX_InGame
 volumeMultiplier=2.5
 concurrencyOverrides {maxCount=2, resolutionRule=StopOldest, retriggerTime=1}
 concurrencySet [&/Conc_TV_Stings]
}
