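# Sound cue for the TV jump-scare boss-spawn sting in the FNM TVScare content.
# It plays a single wave through one wave player on the in-game SFX sound class, with boosted volume and a very large max distance.
# It carries a TV stings concurrency set alongside override values that cap instances, stop the oldest and set a retrigger time.
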SoundCue /FNM_TVScare/Assets/Audio/SC_FNM_TV_JumpScare_BossSpawn
root=WavePlayer_0
nodes [{id=WavePlayer_0, properties {soundWaveAssetPtr=/FNM_TVScare/Assets/Audio/wav/FNM_TV_Sting_JumpScare_SFX_BossSpawn}}]
properties {
 maxDistance=2097152
 soundClassObject=&Fort_SFX_InGame
 volumeMultiplier=2.5
 concurrencyOverrides {maxCount=2, resolutionRule=StopOldest, retriggerTime=1}
 concurrencySet [&/Conc_TV_Stings]
}
