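# Sound cue for the third-person boss-spawn TV jump-scare sting in FNM TVScare.
# A mixer layers a TV glitch wave with a boss spawn wave so both play together, slightly boosted, in the in-game SFX sound class.
# Override attenuation spatializes it with stereo spread, distance falloff and low-pass occlusion.
# It joins the TV stings concurrency set, and its concurrency override settings favor stopping the oldest instance.
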SoundCue /FNM_TVScare/Assets/Audio/SC_FNM_TV_JumpScare_BossSpawn_3P
root=Mixer_0
nodes [
 {id=Mixer_0, children [WavePlayer_1,WavePlayer_3], properties {inputVolume [1,1]}}
 {id=WavePlayer_1, properties {soundWaveAssetPtr=/FNM_TVScare/Assets/Audio/wav/FNM_TV_Sting_JumpScare_SFX_TVGlitch}}
 {id=WavePlayer_3, properties {soundWaveAssetPtr=/FNM_TVScare/Assets/Audio/wav/FNM_TV_Sting_JumpScare_SFX_BossSpawn}}
]
properties {
 bOverrideAttenuation=true
 maxDistance=3250
 soundClassObject=&Fort_SFX_InGame
 volumeMultiplier=1.2
 attenuationOverrides {bEnableOcclusion=true, falloffDistance=2850, occlusionLowPassFilterFrequency=3000, stereoSpread=400}
 concurrencyOverrides {maxCount=2, resolutionRule=StopOldest, retriggerTime=1}
 concurrencySet [&/Conc_TV_Stings]
}
