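# Sound cue for the TV scare's loot-get sting, played as a spatialized in-game effect.
# A mixer layers the completion sting with a randomly picked treasure chest sweetener that enters after a short fixed delay.
# An attenuation node overrides falloff with distance low-pass filtering and occlusion tracing.
# Concurrency limits it to a couple of instances, stopping the oldest.
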SoundCue /FNM_TVScare/Assets/Audio/SC_FNM_TV_Sting_GetLoot
root=Attenuation_0
nodes [
 {
  id=Attenuation_0
  children [Mixer_0]
  properties {
   bOverrideAttenuation=true
   attenuationOverrides {
    bAttenuateWithLPF=true
    bEnableOcclusion=true
    lPFFrequencyAtMax=1000
    lPFRadiusMax=2000
    lPFRadiusMin=200
    occlusionLowPassFilterFrequency=1000
    occlusionTraceChannel=ECC_GameTraceChannel6
    stereoSpread=500
   }
  }
 }
 {id=Delay_0, children [Random_0], properties {delayMax=0.4, delayMin=0.4}}
 {id=Mixer_0, children [WavePlayer_1,Delay_0], properties {inputVolume [0.8,0.7]}}
 {id=Random_0, children [WavePlayer_2,WavePlayer_3], properties {weights [1,1]}}
 {id=WavePlayer_1, properties {soundWaveAssetPtr=/FNM_TVScare/Assets/Audio/wav/FNM_TV_Sting_Interact_Complete_GetLoot}}
 {id=WavePlayer_2, properties {soundWaveAssetPtr=/Game/Sounds/Foley_Loot/Containers/Treasure_Chest/treasure_chest_sweet_shorter_02}}
 {id=WavePlayer_3, properties {soundWaveAssetPtr=/Game/Sounds/Foley_Loot/Containers/Treasure_Chest/treasure_chest_sweet_shorter_03}}
]
properties {
 bOverrideConcurrency=true
 maxDistance=4000
 soundClassObject=&Fort_SFX_InGame
 volumeMultiplier=1.6
 concurrencyOverrides {maxCount=2, resolutionRule=StopOldest, retriggerTime=1}
}
