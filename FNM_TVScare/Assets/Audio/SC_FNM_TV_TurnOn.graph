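# Sound cue for a TV switching on in the FNM TV scare, layering a turn-on one-shot with looping static.
# A mixer blends the one-shot with the static loop, whose enveloper fades it up from silence over about a second.
# It uses the TV static attenuation and in-game SFX class, with a concurrency override that stops the oldest instance once its small voice limit is exceeded.
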SoundCue /FNM_TVScare/Assets/Audio/SC_FNM_TV_TurnOn
root=Mixer_0
nodes [
 {
  id=Enveloper_0
  children [WavePlayer_2]
  properties {
   pitchCurve {editorCurveData {keys [{time=0}]}}
   volumeCurve {editorCurveData {keys [
      {arriveTangent=2.8927, interpMode=RCIM_Cubic, leaveTangent=2.8927, tangentMode=RCTM_User, time=0, value=0}
      {arriveTangent=1.8169, interpMode=RCIM_Cubic, leaveTangent=1.8169, tangentMode=RCTM_User, time=0.9606, value=0.7419}
     ]}}
  }
 }
 {id=Mixer_0, children [WavePlayer_0,Enveloper_0], properties {inputVolume [1.2,1.3]}}
 {id=WavePlayer_0, properties {soundWaveAssetPtr=/FNM_TVScare/Assets/Audio/wav/FNM_TV_TurnOn}}
 {id=WavePlayer_2, properties {bLooping=true, soundWaveAssetPtr=/FNM_TVScare/Assets/Audio/wav/FNM_TV_Static_loop}}
]
properties {
 attenuationSettings=&/AS_FNM_TV_Static
 bOverrideConcurrency=true
 maxDistance=1900
 soundClassObject=&Fort_SFX_InGame
 volumeMultiplier=3
 concurrencyOverrides {maxCount=2, resolutionRule=StopOldest, retriggerTime=1}
}
